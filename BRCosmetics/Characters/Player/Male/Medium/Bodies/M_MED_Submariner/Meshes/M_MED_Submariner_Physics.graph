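# Physics asset for the medium male Submariner player body cosmetic, covering its dangling zippers and a mini quiver.
# Box bodies on the dyn_ bones hang from capsule bodies on spine_05 and thigh_l.
# Left and right zippers chain through tiny master boxes, which rotate freely against the spine, while the center zipper attaches to it directly with angular limits.
# The quiver is held nearly rigid to the left thigh, with only slight twist.
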PhysicsAsset /BRCosmetics/Characters/Player/Male/Medium/Bodies/M_MED_Submariner/Meshes/M_MED_Submariner_Physics
bodies {
 dyn_miniQuiver [{shapeType=Box, extentX=21.5726, extentY=15.3067, extentZ=5.8055, center=<1.1 0 -2>}]
 dyn_zipper [{shapeType=Box, extentX=4.8094, extentY=2.0346, extentZ=0.5099, center=<2.2086 0 -0.0001>}]
 dyn_zipperMaster_l [{shapeType=Box, extentX=0.1, extentY=0.1, extentZ=0.1}]
 dyn_zipperMaster_r [{shapeType=Box, extentX=0.1, extentY=0.1, extentZ=0.1, rotation=<rot 0 -90 0>}]
 dyn_zipper_l [{shapeType=Box, extentX=1.3484, extentY=0.3197, extentZ=3.2021, center=<1.47 0.0009 0.0001>, rotation=<rot 0 -90.032 -89.8936>}]
 dyn_zipper_r [{shapeType=Box, extentX=3.2009, extentY=1.3372, extentZ=0.3556, center=<1.4699 0 0>}]
 spine_05 [{shapeType=Capsule, length=11.2961, radius=9.8495, center=<-0.4961 1.5535 0>}]
 thigh_l [{shapeType=Capsule, length=8.1996, radius=7.3671, center=<-34.7095 0.4278 -0.2>, rotation=<rot 90 0 0>}]
}
constraints [
 {bone1Name=dyn_miniQuiver, bone2Name=thigh_l, swing1LimitDegrees=0, swing2LimitDegrees=0, twistLimitDegrees=5}
 {bone1Name=dyn_zipper_r, bone2Name=dyn_zipperMaster_r, swing1LimitDegrees=75, swing2LimitDegrees=45, twistLimitDegrees=15}
 {bone1Name=dyn_zipper, bone2Name=spine_05, swing1LimitDegrees=45, swing2LimitDegrees=45, twistLimitDegrees=15}
 {bone1Name=dyn_zipper_l, bone2Name=dyn_zipperMaster_l, swing1LimitDegrees=75, swing2LimitDegrees=45, twistLimitDegrees=15}
 {
  bone1Name=dyn_zipperMaster_l
  bone2Name=spine_05
  swing1LimitDegrees=45
  swing1Motion=Free
  swing2LimitDegrees=45
  swing2Motion=Free
  twistLimitDegrees=45
  twistMotion=Free
 }
 {
  bone1Name=dyn_zipperMaster_r
  bone2Name=spine_05
  swing1LimitDegrees=45
  swing1Motion=Free
  swing2LimitDegrees=45
  swing2Motion=Free
  twistLimitDegrees=45
  twistMotion=Free
 }
]
