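# Post-pose Animation Blueprint for the Medium male Submariner body cosmetic, adding secondary motion to its tubes, zippers and knife sheath.
# It copies the attached parent mesh's pose, then a Pose Driver uses left clavicle poses to drive the tube bones through an RBF pose asset.
# Weighted constraints place both zipper master bones, followed by pelvis-space RigidBody simulation and spring-based AnimDynamics on the sheath.
# A separate dyn_tube6 AnimDynamics node is unconnected; the event graph is empty.
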AnimBlueprint /BRCosmetics/Characters/Player/Male/Medium/Bodies/M_MED_Submariner/Meshes/M_MED_Submariner_AnimBP
target_skeleton=&/M_MED_Submariner_Skeleton
graphs [
 AnimGraph {nodes [
   Anim Dynamics - Bone: dyn_tube6(Alpha=1, ExternalForce="0,0,0") {
    bGravityOverrideInSimSpace=true
    bLinearSpring=true
    bUseGravityOverride=true
    alphaBoolBlend {blendOption=Linear}
    boundBone {boneName=dyn_tube6}
    componentLinearAccScale=<0.5 0.5 0.5>
    gravityOverride {z=-980}
    physicsBodyDefinitions [{
      sphereCollisionRadius=0
      boundBone {boneName=dyn_tube6}
      boxExtents=<3 3 3>
      constraintSetup {angularLimitsMax=<0 0 0>, angularLimitsMin=<0 0 0>, angularTarget=<0 0 0>, linearAxesMax=<1 1 1>, linearAxesMin=<-1 -1 -1>}
      localJointOffset=<1 0 0>
     }]
   }
   n1 = Pose Driver - Source: clavicle_l!(SourcePose:Pose Link Structure=Copy Pose From Mesh() {bUseAttachedParent=true}) -> (Pose:Pose Link Structure) {
    is=title AnimGraphNode_PoseDriver
    poseAsset=&/M_MED_Submariner_RBF_PoseAsset
    onlyDriveBones [
     {boneName=dyn_tube1}
     {boneName=dyn_tube2}
     {boneName=dyn_tube3}
     {boneName=dyn_tube4}
     {boneName=dyn_tube5}
     {boneName=dyn_tube6}
     {boneName=dyn_tube7}
     {boneName=dyn_tube8}
    ]
    poseTargets [
     {
      bApplyCustomCurve=false
      bIsHidden=false
      distanceMethod=DefaultMethod
      drivenName=Base
      functionType=DefaultFunction
      targetScale=1.2
      boneTransforms [{targetRotation=<rot -85.1913 -138.5624 143.0618>, targetTranslation=<4.3575 -0.1466 -2.3223>}]
      customCurve {defaultValue=3.4028234663852886e+38, postInfinityExtrap=RCCE_Constant, preInfinityExtrap=RCCE_Constant, keys []}
      targetRotation=<rot 0 0 0>
     }
     {
      bApplyCustomCurve=false
      bIsHidden=false
      distanceMethod=DefaultMethod
      drivenName=Glider
      functionType=DefaultFunction
      targetScale=0.7
      boneTransforms [{targetRotation=<rot -65 30 -17.1525>, targetTranslation=<4.357 -0.1466 -2.3223>}]
      customCurve {defaultValue=3.4028234663852886e+38, postInfinityExtrap=RCCE_Constant, preInfinityExtrap=RCCE_Constant, keys []}
      targetRotation=<rot 0 0 0>
     }
     {
      bApplyCustomCurve=false
      bIsHidden=false
      distanceMethod=DefaultMethod
      drivenName=PistolUp
      functionType=DefaultFunction
      targetScale=1
      boneTransforms [{targetRotation=<rot -50 10 -5>, targetTranslation=<4.357 -0.1465 -2.3221>}]
      customCurve {
       defaultValue=3.4028234663852886e+38
       postInfinityExtrap=RCCE_Constant
       preInfinityExtrap=RCCE_Constant
       keys [
        {
         arriveTangent=0
         arriveTangentWeight=0
         interpMode=RCIM_Linear
         leaveTangent=0
         leaveTangentWeight=0
         tangentMode=RCTM_Auto
         tangentWeightMode=RCTWM_WeightedNone
         time=0
         value=0
        }
        {
         arriveTangent=0
         arriveTangentWeight=0
         interpMode=RCIM_Linear
         leaveTangent=0
         leaveTangentWeight=0
         tangentMode=RCTM_Auto
         tangentWeightMode=RCTWM_WeightedNone
         time=1
         value=1
        }
       ]
      }
      targetRotation=<rot 0 0 0>
     }
     {
      bApplyCustomCurve=false
      bIsHidden=false
      distanceMethod=DefaultMethod
      drivenName=PistolDown
      functionType=DefaultFunction
      targetScale=0.4232
      boneTransforms [{targetRotation=<rot -85.5985 81.607 -79.3271>, targetTranslation=<4.3529 -0.1465 -2.3221>}]
      customCurve {defaultValue=3.4028234663852886e+38, postInfinityExtrap=RCCE_Constant, preInfinityExtrap=RCCE_Constant, keys []}
      targetRotation=<rot 0 0 0>
     }
     {
      bApplyCustomCurve=false
      bIsHidden=false
      distanceMethod=DefaultMethod
      drivenName=MinigunUp
      functionType=DefaultFunction
      targetScale=0.9387
      boneTransforms [{targetRotation=<rot -49.2935 -31.1941 19.5046>, targetTranslation=<4.3538 -0.1458 -2.3224>}]
      customCurve {defaultValue=3.4028234663852886e+38, postInfinityExtrap=RCCE_Constant, preInfinityExtrap=RCCE_Constant, keys []}
      targetRotation=<rot 0 0 0>
     }
     {
      bApplyCustomCurve=false
      bIsHidden=false
      distanceMethod=DefaultMethod
      drivenName=MinigunDown
      functionType=DefaultFunction
      targetScale=1
      boneTransforms [{targetRotation=<rot -69.2494 100.6789 -95.0214>, targetTranslation=<4.3461 -0.1466 -2.3226>}]
      customCurve {defaultValue=3.4028234663852886e+38, postInfinityExtrap=RCCE_Constant, preInfinityExtrap=RCCE_Constant, keys []}
      targetRotation=<rot 0 0 0>
     }
     {
      bApplyCustomCurve=false
      bIsHidden=false
      distanceMethod=DefaultMethod
      drivenName=PistolForward
      functionType=DefaultFunction
      targetScale=0.4232
      boneTransforms [{targetRotation=<rot -81.1166 29.9743 -26.3605>, targetTranslation=<4.3528 -0.1465 -2.3225>}]
      customCurve {defaultValue=3.4028234663852886e+38, postInfinityExtrap=RCCE_Constant, preInfinityExtrap=RCCE_Constant, keys []}
      targetRotation=<rot 0 0 0>
     }
     {
      bApplyCustomCurve=false
      bIsHidden=false
      distanceMethod=DefaultMethod
      drivenName=DanceEnd
      functionType=DefaultFunction
      targetScale=0.7
      boneTransforms [{targetRotation=<rot -65 -9.5866 2.7341>, targetTranslation=<4.3572 -0.1465 -2.3223>}]
      customCurve {defaultValue=3.4028234663852886e+38, postInfinityExtrap=RCCE_Constant, preInfinityExtrap=RCCE_Constant, keys []}
      targetRotation=<rot 0 0 0>
     }
    ]
    rBFParams {radius=16}
    sourceBones [{boneName=clavicle_l}]
   }
   n2 = "Constraint - dyn_zipperMaster_l (PPPPP)"!(
    ComponentPose=Local To Component(LocalPose=@n1)
    Alpha=1
    ConstraintWeights_0=0.329
    ConstraintWeights_1:float=0.0014
    ConstraintWeights_2:float=0.3919
    ConstraintWeights_3:float=0.0649
    ConstraintWeights_4:float=0.2127
   ) -> (Pose) {
    alphaBoolBlend {blendOption=Linear}
    boneToModify {boneName=dyn_zipperMaster_l}
    constraintSetup [
     {targetBone {boneName=upperarm_twist_02_l}}
     {targetBone {boneName=spine_03}}
     {targetBone {boneName=upperarm_twist_01_l}}
     {targetBone {boneName=clavicle_l}}
     {targetBone {boneName=upperarm_l}}
    ]
    constraintWeights [1,1,1,1,1]
   }
   n3 = "Constraint - dyn_zipperMaster_r (PPPPP)"!(
    ComponentPose=@n2
    Alpha=1
    ConstraintWeights_0=0.4168
    ConstraintWeights_1:float=0.0768
    ConstraintWeights_2:float=0.0012
    ConstraintWeights_3:float=0.1931
    ConstraintWeights_4:float=0.3122
   ) -> (Pose) {
    alphaBoolBlend {blendOption=Linear}
    boneToModify {boneName=dyn_zipperMaster_r}
    constraintSetup [
     {targetBone {boneName=upperarm_twist_01_r}}
     {targetBone {boneName=clavicle_r}}
     {targetBone {boneName=spine_03}}
     {targetBone {boneName=upperarm_r}}
     {targetBone {boneName=upperarm_twist_02_r}}
    ]
    constraintWeights [1,1,1,1,1]
   }
   n4 = RigidBody(ComponentPose=@n3) {
    bClampLinearTranslationLimitToRefPose=true
    bOverrideWorldGravity=true
    overridePhysicsAsset=&/M_MED_Submariner_Physics
    simulationSpace=BaseBoneSpace
    alphaBoolBlend {blendOption=Linear}
    baseBoneRef {boneName=pelvis}
    componentLinearAccScale=<0.1 0.1 0.1>
    overrideWorldGravity {z=-980}
   }
   n5 = Anim Dynamics - Bone: dyn_knifeSheath(ComponentPose=@n4, Alpha=1, ExternalForce="0,0,0") {
    angularSpringConstant=1000
    bAngularSpring=true
    bGravityOverrideInSimSpace=true
    bUseGravityOverride=true
    gravityScale=-1
    simulationSpace=RootRelative
    alphaBoolBlend {blendOption=Linear}
    boundBone {boneName=dyn_knifeSheath}
    componentAppliedLinearAccClamp=<50 50 50>
    componentLinearAccScale=<0.5 0.5 0.5>
    componentLinearVelScale=<0.5 0.5 0.5>
    gravityOverride {z=-980}
    physicsBodyDefinitions [{
      sphereCollisionRadius=0
      boundBone {boneName=dyn_knifeSheath}
      boxExtents=<10 10 10>
      constraintSetup {angularLimitsMax=<5 2 5>, angularLimitsMin=<-5 -2 -5>, angularTarget=<0 0 0>}
      localJointOffset=<6 0 0>
     }]
    relativeSpaceBone {boneName=root}
   }
   Output Pose(Result=Component To Local(ComponentPose=@n5)) {layerGroup=None}
  ]}
 EventGraph {nodes [Event BlueprintUpdateAnimation()]}
]
variables []
